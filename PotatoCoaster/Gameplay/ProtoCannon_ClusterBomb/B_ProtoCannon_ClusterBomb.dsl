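; Weapon Blueprint for the PotatoCoaster ProtoCannon cluster bomb, handling charge-up effects.
; While charging, it applies a Charging gameplay effect to the instigator and, for a local viewing player, starts a camera shake.
; Looping timers feed charge percentage into the mesh material and rewrite the live shake's oscillators from a curve over charge time, accumulating amplitude each tick.
; Ending charge removes the effect, stops the shake and resets the material.
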(event EventOnStartCharge
  (CallFunction|StartRecoil)
  (bind _returnvalue (GameplayEffects|ApplyGameplayEffectToSelf :self (Ability|GetFortAbilitySystemComponent (Game|GetInstigator)) :GameplayEffectClass "/PotatoCoaster/Gameplay/GAS/GE_PotatoCoaster_ProtoCannon_Charging.GE_PotatoCoaster_ProtoCannon_Charging_C" :Level 0.0))
  (Variables|Default|SetGE_ChargingHanlde _returnvalue))

(event EventOnEndCharge
  (CallFunction|EndRecoil)
  (GameplayEffects|RemoveActiveGameplayEffect (Ability|GetFortAbilitySystemComponent (Game|GetInstigator)) (Variables|Default|GetGE_ChargingHanlde) -1))

(event Custom|WeaponMeshUpdate
  (bind _returnvalue (FortWeapon|GetChargePercent))
  (Rendering|Material|SetScalarParameterValueonMaterials (Variables|Visual|GetWeaponMesh) "WeaponChargePercentage" _returnvalue)
  (if (== _returnvalue 1.0)
    (Utilities|Time|ClearandInvalidateTimerbyHandle (Variables|Default|GetWeaponMeshUpdateTimerHandle))))

(event Custom|Start%20Recoil
  (bind _returnvalue (Game|GetInstigator))
  (bind _returnvalue_2 (Pawn|GetLocalViewingPlayerController _returnvalue))
  (Utilities|IsValid _returnvalue
    (:Is%20Valid
      (Utilities|IsValid _returnvalue_2
        (:Is%20Valid
          (bind _playercameramanager (Variables|/Script/Engine.PlayerController|GetPlayerCameraManager _returnvalue_2)
            (:then
              (bind _returnvalue_3 (CameraShakes|StartCameraShake :self _playercameramanager :ShakeClass "/PotatoCoaster/Gameplay/ProtoCannon_ClusterBomb/Impulse/B_ProtoCannon_Laser_CameraShake.B_ProtoCannon_Laser_CameraShake_C" :Scale 1.0 :PlaySpace "CameraLocal"))
              (Variables|CameraShake|SetCameraShake (CameraShakes|ConvLegacyCameraShake _returnvalue_3))
              (bind _returnvalue_4 (Utilities|Time|SetTimerbyEvent (EventDispatchers|CreateEvent :!event Custom|CameraShakeTick) (ScalableFloat|GetValueatLevel (Variables|CameraShake|GetCameraShakeTickRate)) true))
              (Variables|CameraShake|SetCameraShakeTimerHandle _returnvalue_4)
              (bind _returnvalue_5 (Utilities|Time|SetTimerbyEvent (EventDispatchers|CreateEvent :!event Custom|DoOnChargeTime) (ScalableFloat|GetValueatLevel (Variables|CameraShake|GetHF%20Max%20Charge%20Time))))
              (Variables|CameraShake|SetChargeTimeHandle _returnvalue_5)
              (bind _returnvalue_6 (Utilities|Time|SetTimerbyEvent (EventDispatchers|CreateEvent :!event Custom|WeaponMeshUpdate) (ScalableFloat|GetValueatLevel (Variables|Default|GetWeaponMeshParamTickRate)) true true))
              (Variables|Default|SetWeaponMeshUpdateTimerHandle _returnvalue_6))
            (:else)))))))

(event Custom|CameraShakeTick
  (bind _camerashake (Variables|CameraShake|GetCameraShake))
  (bind (_rotoscillation_pitch _rotoscillation_yaw _rotoscillation_roll) (Utilities|Struct|BreakROscillator (Variables|/Script/EngineCameras.LegacyCameraShake|GetRotOscillation_Pitch_Amplitude _camerashake)))
  (bind _rotoscillation_pitch_amplitude (Utilities|Struct|BreakFOscillator _rotoscillation_pitch))
  (bind _rotoscillation_yaw_amplitude (Utilities|Struct|BreakFOscillator _rotoscillation_yaw))
  (bind _rotoscillation_roll_amplitude (Utilities|Struct|BreakFOscillator _rotoscillation_roll))
  (bind _fovoscillation_amplitude (Utilities|Struct|BreakFOscillator (Variables|/Script/EngineCameras.LegacyCameraShake|GetFOVOscillation_Amplitude _camerashake)))
  (Variables|CameraShake|SetCurrentCurveValue (Animation|GetFloatValuefromCurve (Variables|CameraShake|GetCameraShake_FloatCurve) (/ (Utilities|Time|GetTimerElapsedTimebyHandle (Variables|CameraShake|GetChargeTimeHandle)) (ScalableFloat|GetValueatLevel (Variables|CameraShake|GetHF%20Max%20Charge%20Time)))))
  (Variables|/Script/EngineCameras.LegacyCameraShake|SetRotOscillation_Pitch_Amplitude (Utilities|Struct|MakeROscillator (Utilities|Struct|MakeFOscillator (CallFunction|CalculateAmplitude _rotoscillation_pitch_amplitude) (CallFunction|CalculateFrequency) "PerlinNoise") (Utilities|Struct|MakeFOscillator (CallFunction|CalculateAmplitude _rotoscillation_yaw_amplitude) (CallFunction|CalculateFrequency) "PerlinNoise") (Utilities|Struct|MakeFOscillator (CallFunction|CalculateAmplitude _rotoscillation_roll_amplitude) (CallFunction|CalculateFrequency) "PerlinNoise")) _camerashake)
  (Variables|/Script/EngineCameras.LegacyCameraShake|SetFOVOscillation_Amplitude (Utilities|Struct|MakeFOscillator (CallFunction|CalculateAmplitude _fovoscillation_amplitude) (* (Variables|CameraShake|GetCurrentCurveValue) (ScalableFloat|GetValueatLevel (Variables|CameraShake|GetCameraShakeFOVMax))) "PerlinNoise") _camerashake))

(event Custom|EndRecoil
  (bind _returnvalue (Game|GetInstigator))
  (bind _returnvalue_2 (Pawn|GetLocalViewingPlayerController _returnvalue))
  (Utilities|Time|ClearandInvalidateTimerbyHandle (Variables|CameraShake|GetCameraShakeTimerHandle))
  (Utilities|Time|ClearandInvalidateTimerbyHandle (Variables|Default|GetWeaponMeshUpdateTimerHandle))
  (Rendering|Material|SetScalarParameterValueonMaterials (Variables|Visual|GetWeaponMesh) "WeaponChargePercentage" 0.0)
  (Utilities|IsValid _returnvalue
    (:Is%20Valid
      (Utilities|IsValid _returnvalue_2
        (:Is%20Valid
          (bind _playercameramanager (Variables|/Script/Engine.PlayerController|GetPlayerCameraManager _returnvalue_2)
            (:then
              (CameraShakes|StopCameraShake _playercameramanager (Variables|CameraShake|GetCameraShake) true))
            (:else)))))))

(event Custom|DoOnChargeTime
  (Utilities|Time|ClearandInvalidateTimerbyHandle (Variables|CameraShake|GetChargeTimeHandle)))

(event Custom|CheckRemoveWeaponOnEmpty)

(fn CalculateAmplitude
  (params (PrevAmplitude :type Float%20%28double-precision%29))
  (return (+ PrevAmplitude (* (Variables|CameraShake|GetCurrentCurveValue) (ScalableFloat|GetValueatLevel (Variables|CameraShake|GetCameraShakeAmplitudeTickMax))))))

(fn CalculateFrequency
  (return (* (Variables|CameraShake|GetCurrentCurveValue) (ScalableFloat|GetValueatLevel (Variables|CameraShake|GetCameraShakeFrequencyMax)))))

(fn ConstructionScript
  (|CallParentFunction))
